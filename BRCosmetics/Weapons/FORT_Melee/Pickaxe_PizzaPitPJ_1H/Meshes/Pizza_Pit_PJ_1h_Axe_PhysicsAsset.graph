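# Physics asset for the one-handed Pizza Pit PJ pickaxe, a cosmetic melee weapon, defining its collision and a rotating blade joint.
# A capsule covers the handle, a thin square box stands in for dyn_blade, and a small cube sits on dyn_blade_MSTR.
# The only constraint links blade to master with both swings locked and twist free, so the blade can spin about one axis like a wheel.
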PhysicsAsset /BRCosmetics/Weapons/FORT_Melee/Pickaxe_PizzaPitPJ_1H/Meshes/Pizza_Pit_PJ_1h_Axe_PhysicsAsset
bodies {
 Melee_TwoHanded_Handle [{shapeType=Capsule, length=33.6665, radius=4.2852, center=<1.3953 -7.8442 36.8274>, rotation=<rot 0.013 -89.9999 3.0104>}]
 dyn_blade [{shapeType=Box, extentX=20, extentY=0.5, extentZ=20, center=<-1 0 0>, rotation=<rot 90 0 0>}]
 dyn_blade_MSTR [{shapeType=Box, extentX=7.5, extentY=7.5, extentZ=7.5}]
}
constraints [{
  bone1Name=dyn_blade
  bone2Name=dyn_blade_MSTR
  swing1LimitDegrees=45
  swing1Motion=Locked
  swing2LimitDegrees=45
  swing2Motion=Locked
  twistLimitDegrees=45
  twistMotion=Free
 }]
